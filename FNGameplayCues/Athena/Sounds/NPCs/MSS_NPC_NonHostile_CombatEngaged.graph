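# One-shot stereo MetaSound source that plays the combat-engaged sound for non-hostile NPCs.
# OnPlay triggers a single stereo Wave Player with a fixed wave asset, feeding both outputs and signaling OnFinished when done.
# The In Array wave input is declared but not connected.
# It uses NPC alert attenuation, the critical in-game UI sound class, and a concurrency override that blocks new instances at the limit.
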MetaSoundSource /FNGameplayCues/Athena/Sounds/NPCs/MSS_NPC_NonHostile_CombatEngaged
inputs [
 In Array {type=WaveAsset:Array}
 UE.Source.OnPlay {type=Trigger}
]
interfaces [UE.OutputFormat.Stereo:1.0, UE.Source.OneShot:1.0, UE.Source:1.1]
outputs [
 "UE.OutputFormat.Stereo.Audio:0" {type=Audio}
 "UE.OutputFormat.Stereo.Audio:1" {type=Audio}
 UE.Source.OneShot.OnFinished {type=Trigger}
]
pages [Default {
  input_defaults {UE.Source.OnPlay=null, In Array [&/NPC_NonHostile_CombatEngaged]}
  nodes [WavePlayer_0 = UE.Wave Player:Stereo(Play=^UE.Source.OnPlay) {also {defaults {Wave Asset=&NPC_NonHostile_CombatEngaged}, version [1,1]}}]
  links [
   {dst="@UE.Source.OneShot.OnFinished", src=WavePlayer_0.On Finished}
   {dst="@UE.OutputFormat.Stereo.Audio:0", src=WavePlayer_0.Out Left}
   {dst="@UE.OutputFormat.Stereo.Audio:1", src=WavePlayer_0.Out Right}
  ]
 }]
properties {
 assetImportData=&self:AssetImportData
 attenuationSettings=&/NPC_Alert_Att
 bOverrideConcurrency=true
 maxDistance=9000
 numChannels=2
 outputFormat=Stereo
 sampleRate=48000
 soundClassObject=&Fort_UI_InGame_Critical
 concurrencyOverrides {maxCount=3, resolutionRule=PreventNew}
}
