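# Sound cue for the NPC alerted-state "question mark" sound, classed as critical in-game UI audio.
# A single wave player plays the wave with boosted volume, using NPC alert attenuation and concurrency assets.
# It sends to sidechain volume buses for music, first-person weapons, weapons and explosions, apparently so those mixes can duck under the alert.
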SoundCue /FNGameplayCues/Athena/Sounds/NPCs/NPC_Alert_Cue
root=WavePlayer_0
nodes [{id=WavePlayer_0, properties {soundWaveAssetPtr=&/NPC_State_Alerted_Question_Mark}}]
properties {
 attenuationSettings=&/NPC_Alert_Att
 maxDistance=9000
 soundClassObject=&Fort_UI_InGame_Critical
 volumeMultiplier=1.4
 busSends [
  {
   audioBus=&AB_Sidechain_FortMusic_Vol
   maxSendLevel=0.3
   minSendLevel=0
   sendLevel=0.25
   soundSourceBus=null
  }
  {audioBus=/Game/Sounds/Modulation/EnvelopeFollowers/Volume/AB_Sidechain_FortWeapon1P_Vol, minSendLevel=0, sendLevel=0.25, soundSourceBus=null}
  {audioBus=&AB_Sidechain_FortWeapon_Vol, minSendLevel=0, sendLevel=0.25, soundSourceBus=null}
  {
   audioBus=&AB_Sidechain_FortExplosion_Vol
   minSendLevel=0
   sendLevel=0.25
   soundSourceBus=null
  }
 ]
 concurrencyOverrides {bLimitToOwner=true, maxCount=1}
 concurrencySet [&/NPC_Alert_Conc]
}
