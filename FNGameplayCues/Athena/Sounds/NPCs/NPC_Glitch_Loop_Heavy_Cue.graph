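# Sound cue for the heavy NPC glitch effect loop, played as quiet, pitched-down spatialized audio in the player foley sound class.
# A single looping wave player feeds every branch of a quality-level node, so all quality settings play the same wave, then passes through a modulator at default settings.
# The cue uses shared NPC glitch attenuation and overrides concurrency to cap simultaneous instances at two with volume scaling.
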SoundCue /FNGameplayCues/Athena/Sounds/NPCs/NPC_Glitch_Loop_Heavy_Cue
root=Modulator_0
nodes [
 {id=Modulator_0, children [QualityLevel_0]}
 {id=QualityLevel_0, children [WavePlayer_0,WavePlayer_0,WavePlayer_0,WavePlayer_0]}
 {id=WavePlayer_0, properties {bLooping=true, soundWaveAssetPtr=&/NPC_Glitch_VFX_Heavy_Loop}}
]
properties {
 attenuationSettings=/Game/Athena/Sounds/NPCs/NPC_Glitch_Att
 bOverrideConcurrency=true
 maxDistance=1500
 pitchMultiplier=0.8
 soundClassObject=&Fort_Player_Foley_Athena
 volumeMultiplier=0.3
 concurrencyOverrides {maxCount=2, volumeScale=0.5}
}
