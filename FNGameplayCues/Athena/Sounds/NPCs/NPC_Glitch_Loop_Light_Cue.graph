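# Sound cue for the light variant of the looping NPC glitch sound.
# A single looping wave player of the glitch VFX loop feeds all four quality-level inputs, then passes through a modulator at the root.
# The cue uses NPC glitch attenuation and the player foley sound class at reduced volume.
# A concurrency override caps simultaneous instances at two with volume scaling.
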SoundCue /FNGameplayCues/Athena/Sounds/NPCs/NPC_Glitch_Loop_Light_Cue
root=Modulator_0
nodes [
 {id=Modulator_0, children [QualityLevel_0]}
 {id=QualityLevel_0, children [WavePlayer_26, WavePlayer_26, WavePlayer_26, WavePlayer_26]}
 {id=WavePlayer_26, properties {bLooping=true, soundWaveAssetPtr=/Game/Athena/Sounds/NPCs/NPC_Glitch_VFX_Loop}}
]
properties {
 attenuationSettings=/Game/Athena/Sounds/NPCs/NPC_Glitch_Att
 bOverrideConcurrency=true
 maxDistance=1500
 soundClassObject=&Fort_Player_Foley_Athena
 volumeMultiplier=0.3155
 concurrencyOverrides {maxCount=2, volumeScale=0.5}
}
