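# Sound cue for an NPC teleporting in, mixing two glitch effect layers on the gadgets sound class.
# The spawn glitch layer plays slightly quieter after a short random delay, while the second glitch layer starts immediately through a volume envelope that holds, then fades to silence.
# Both layers get slight random pitch variation.
# The cue uses NPC teleport attenuation and overrides concurrency to cap simultaneous instances.
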SoundCue /FNGameplayCues/Athena/Sounds/NPCs/NPC_Teleport_In_Cue
root=Mixer_1
nodes [
 {id=Delay_0, children [Modulator_2], properties {delayMax=0.2, delayMin=0.1}}
 {
  id=Enveloper_0
  children [WavePlayer_1]
  properties {pitchCurve {editorCurveData {keys [{time=0}]}}, volumeCurve {editorCurveData {keys [<at 0 1>,<at 0.7699 1.0015>,<at 0.9576 -0.0001>]}}}
 }
 {id=Mixer_1, children [Delay_0,Modulator_3], properties {inputVolume [0.7,1]}}
 {id=Modulator_2, children [WavePlayer_0], properties {pitchMax=1.1, pitchMin=0.98}}
 {id=Modulator_3, children [Enveloper_0], properties {pitchMax=1.1, pitchMin=0.98}}
 {id=WavePlayer_0, properties {soundWaveAssetPtr=/Game/Athena/Sounds/NPCs/NPC_Glitch_VFX_Spawn}}
 {id=WavePlayer_1, properties {soundWaveAssetPtr=/Game/Athena/Sounds/NPCs/NPC_Glitch_VFX_13}}
]
properties {
 attenuationSettings=/Game/Athena/Sounds/NPCs/NPC_Teleport_Att
 bOverrideConcurrency=true
 maxDistance=8500
 pitchMultiplier=1.05
 soundClassObject=&Fort_Gadgets
 volumeMultiplier=0.25
 concurrencyOverrides {maxCount=2}
}
