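# Sound cue for an NPC teleporting out, built from two layered NPC glitch waves and routed to the Fort_Gadgets sound class.
# A mixer blends a pitch-raised, quieter spawn glitch with a death glitch whose volume is shaped by an envelope that rises quickly, holds, then fades.
# A root modulator at default settings wraps the mix, which uses a shared teleport attenuation asset and its own low concurrency limit.
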SoundCue /FNGameplayCues/Athena/Sounds/NPCs/NPC_Teleport_Out_Cue
root=Modulator_1
nodes [
 {
  id=Enveloper_0
  children [WavePlayer_0]
  properties {
   pitchCurve {editorCurveData {keys [{time=0}]}}
   volumeCurve {editorCurveData {keys [<at 0.0088 0.0305>,<at 0.0607 0.9973>,<at 0.4115 0.9971>,<at 1 0>]}}
  }
 }
 {id=Mixer_0, children [Modulator_0,Enveloper_0], properties {inputVolume [0.55,1]}}
 {id=Modulator_0, children [WavePlayer_2], properties {pitchMax=1.15, pitchMin=1.1, volumeMax=1, volumeMin=1}}
 {id=Modulator_1, children [Mixer_0]}
 {id=WavePlayer_0, properties {soundWaveAssetPtr=/Game/Athena/Sounds/NPCs/NPC_Glitch_VFX_Death}}
 {id=WavePlayer_2, properties {soundWaveAssetPtr=/Game/Athena/Sounds/NPCs/NPC_Glitch_VFX_Spawn}}
]
properties {
 attenuationSettings=/Game/Athena/Sounds/NPCs/NPC_Teleport_Att
 bOverrideConcurrency=true
 maxDistance=8500
 pitchMultiplier=1.05
 soundClassObject=&Fort_Gadgets
 volumeMultiplier=0.35
 attenuationOverrides {bAttenuate=false}
 concurrencyOverrides {maxCount=2}
}
